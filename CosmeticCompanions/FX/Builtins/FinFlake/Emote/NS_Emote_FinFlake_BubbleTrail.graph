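# Niagara ribbon trail for the FinFlake cosmetic companion's BubbleTrail emote, built from one continuously spawning Tail_Trail emitter.
# Looped emitter age drives a lifetime curve that shortens the trail in two windows, and a user parameter sets ribbon width.
# Over particle life, width swells then tapers, alpha fades, and upward force kicks in near the end.
# Time-of-day color and scalar inputs tint and dim it; User.Emote_FinFlake_Lifetime appears unused.
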NiagaraSystem /CosmeticCompanions/FX/Builtins/FinFlake/Emote/NS_Emote_FinFlake_BubbleTrail
dependencies {
 usedDataInterfaces [&NiagaraDataInterfaceCurve]
 usedDynamicInputs [&FNTimeOfDayColor]
 usedModules [
  &self:SystemSpawnScript.SystemScriptSource.SystemScriptGraph.NiagaraNodeAssignment_0.SetVariables_#1
  &SystemState
 ]
}
emitters {Tail_Trail {
  emitterUpdateScript [
   &EmitterState()
   &SpawnRate(SpawnRate=30, Use Spawn Probability=false)
  ]
  particleSpawnScript [&InitializeParticle(
    Lifetime=&FloatFromCurve(
     FloatCurve=DataInterface() {propertyValues {Curve {keys [<at 0 1.667>,<at 3.45 1.667>,<at 3.5 0.8>,<at 4.75 0.8>,<at 4.8 1.667>,<at 8.1 1.667>,<at 8.15 0.8>,<at 9.5 0.8>,<at 9.55 1.667>,<at 9.667 1.667>]}}}
     CurveIndex=$Emitter.LoopedAge
     Scale Curve=0.15
    )
    Ribbon Width=$User.Emote_FinFlake_RibbonWidth
    Ribbon Width Mode=::NewEnumerator1
    UsePositionOffset=false
    Write Material Random=true
   )]
  particleUpdateScript [
   &ParticleState()
   &ScaleRibbonWidth(Ribbon Width Scale=&FloatFromCurve(FloatCurve=DataInterface() {propertyValues {Curve {
        defaultValue=1
        keys [
         {arriveTangent=9.0725, interpMode=RCIM_Cubic, leaveTangent=9.0725, tangentMode=RCTM_User, time=0, value=0}
         {arriveTangent=-2.5249, interpMode=RCIM_Cubic, leaveTangent=-2.5249, tangentMode=RCTM_User, time=0.1408, value=0.9847}
         {arriveTangent=-0.0405, interpMode=RCIM_Cubic, leaveTangent=-0.0405, tangentMode=RCTM_User, time=1, value=0}
        ]
       }}}))
   &ScaleColor(
    Scale Alpha=&FloatFromCurve(
     FloatCurve=DataInterface() {propertyValues {Curve {
        defaultValue=1
        keys [{interpMode=RCIM_Cubic, tangentMode=RCTM_User, time=0, value=1}, {interpMode=RCIM_Cubic, tangentMode=RCTM_User, time=1, value=0}]
       }}}
     Scale Curve=&FNTimeOfDayScalar(Night Scale=0.85, Morning Scale=1, Evening Scale=1)
    )
    Scale RGB=&MakeVectorFromLinearColorRGB(Color=$System.TOD)
    ScaleA=true
    ScaleRGB=true
   )
   &GravityForce(Gravity=&MakeVector(
     X=0
     Y=0
     Z=&FloatFromCurve(
      FloatCurve=DataInterface() {propertyValues {Curve {keys [
          <at 0 0>
          {arriveTangent=-0.0206, interpMode=RCIM_Cubic, leaveTangent=-0.0206, tangentMode=RCTM_User, time=0.8, value=0}
          {arriveTangent=7.2281, interpMode=RCIM_Cubic, leaveTangent=7.2281, tangentMode=RCTM_User, time=1, value=1}
         ]}}}
      Scale Curve=300
     )
    ))
   &SolveForcesAndVelocity(Clamp Velocity=false, Limit Acceleration=false)
  ]
  rendererClasses [&NiagaraRibbonRendererProperties]
  renderers [{rendererClass=&NiagaraRibbonRendererProperties}]
 }}
summary {
 systemName=NS_Emote_FinFlake_BubbleTrail
 emitters [{emitterName=Tail_Trail, rendererClasses [&NiagaraRibbonRendererProperties]}]
 userVariables [
  User.Emote_FinFlake_Lifetime {type=NiagaraFloat, defaultValue=0, description=""}
  User.Emote_FinFlake_RibbonWidth {type=NiagaraFloat, defaultValue=0, description=""}
 ]
}
