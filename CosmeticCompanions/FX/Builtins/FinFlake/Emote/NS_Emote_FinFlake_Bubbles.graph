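# Niagara system for the bubbles effect of the FinFlake cosmetic companion's emote, judging by its name and folder.
# It holds one emitter, Bubbles, with no renderers listed, and exposes no user variables.
# System scripts use SystemState and a Set Variables assignment, and the system depends on a Skeletal Mesh data interface.
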NiagaraSystem /CosmeticCompanions/FX/Builtins/FinFlake/Emote/NS_Emote_FinFlake_Bubbles
dependencies {
 usedDataInterfaces [&NiagaraDataInterfaceSkeletalMesh]
 usedDynamicInputs []
 usedModules [
  &self:SystemSpawnScript.SystemScriptSource.SystemScriptGraph.NiagaraNodeAssignment_0.SetVariables_#1
  &SystemState
 ]
}
emitters {Bubbles {rendererClasses [], renderers []}}
summary {systemName=NS_Emote_FinFlake_Bubbles, emitters [{emitterName=Bubbles, rendererClasses []}], userVariables []}
